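# Translucent subsurface master material for liquid effects on Niagara mesh particles.
# Particle color drives HSV-shaded diffuse and tinted subsurface, while emissive fakes a normal-mapped reflection texture and sun specular, with a minimum offset for mobile.
# Opacity combines particle alpha, near-camera fade and a Fresnel shaped by vertex green and dynamic Erode.
# World position offset adds animated curl-noise distortion and vertex-red-masked gravity, alongside refraction and an optional added texture.
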Material /DealershipSharedLibrary/FX/Materials/MasterMaterials/Liquid/M_DLR_Liquid_Mesh
graph {
 properties {
  bAutomaticallySetUsageInEditor=false
  bOutputTranslucentVelocity=true
  bUsedWithNiagaraMeshParticles=true
  blendMode=BLEND_Translucent
  shadingModel=MSM_Subsurface
  usedShadingModels=Subsurface
 }
 nodes [
  n1 = ScalarParameter() {defaultValue=40, group=Opacity, parameterName=NearFade}
  n2 = Multiply(
   A=Power(
    Base=MaterialFunctionCall(Distance=@n1) {materialFunction=&Near_Camera_Fade}.Result
    Exp=ScalarParameter() {defaultValue=3, group=Opacity, parameterName=NearFadeExp}
   )
   B=ParticleColor().A
  )
  DepthFade(Opacity=@n2, FadeDistance=ScalarParameter() {defaultValue=2, group=Opacity, parameterName=Depth Fade Distance})
  MaterialFunctionCall(Emissive=NamedRerouteUsage()) {materialFunction=&MF_FX_EmissiveAdjustments}
  n3 = VertexColor()
  n4 = Multiply(A=@n3.R, B=@n3.R)
  Multiply(A=@n4, B=@n4)
  n5 = VertexNormalWS()
  n6 = Constant() {r=0.5}
  n7 = ComponentMask(None=Add(A=Multiply(A=@n5, B=@n6), B=@n6)) {g=true, r=true}
  n8 = Desaturation(
   None=Multiply(
    A=TextureSample(UVs=@n7) {texture=/Game/VisualThreatManager/StormVisuals/LargeClouds/T_EyeballReflection2}.RGB
    B=ScalarParameter() {defaultValue=0.2, group=Specular, parameterName=Reflection Amount}
   )
   Fraction=ScalarParameter() {group=Specular, parameterName=Reflection - Desaturation}
  )
  n9 = ScalarParameter() {
   defaultValue=0.075
   desc="Offset the Minimum fake specular amount for Mobile.  ("
   group=Specular
   parameterName=Mobile Specular Min Offset
   sliderMax=0.2
  }
  n10 = FeatureLevelSwitch(Default=Constant(), ES3_1=@n9)
  n11 = CollectionParameter() {
   collection=&FortniteMaterialParameters
   parameterName=SunAndMoonModelDirectionalVector
  }
  n12 = ComponentMask(None=@n11) {b=true, g=true, r=true}
  n13 = ReflectionVectorWS(CustomWorldNormal=@n5) {bNormalizeCustomWorldNormal=true}
  n14 = Power(Base=Saturate(None=DotProduct(A=@n12, B=@n13)), Exp=ScalarParameter() {defaultValue=40, group=Specular, parameterName=Specular Power}) {constExponent=40}
  n15 = ScalarParameter() {defaultValue=5, group=Specular, parameterName=Specular Intensity}
  n16 = Power(Base=ComponentMask(None=CollectionParameter() {collection=&FortniteMaterialParameters, parameterName=SunLightColor}) {b=true, g=true, r=true}) {constExponent=0.1}
  n17 = Add(
   A=Multiply(
    A=Clamp(None=@n8, Min=@n10) {minDefault=0.1}
    B=MaterialFunctionCall() {materialFunction=&Color_Time_of_Day}.Result
   )
   B=Multiply(A=Clamp(None=@n14, Min=@n10, Max=@n15), B=Clamp(None=@n16))
  ) {constB=0}
  n18 = Add(A=Multiply(A=NamedRerouteUsage(), B=ScalarParameter() {defaultValue=0.1, parameterName=DiffuseColorEmissiveMultiply}), B=@n17)
  n19 = NamedRerouteUsage()
  n20 = ParticleColor()
  NamedRerouteDeclaration(None=Add(A=StaticSwitchParameter(True=@n18, False=@n17) {parameterName=DiffuseColor on Emissive?}, B=Multiply(A=@n19, B=@n20.A) {constB=0})) {name=Emissive, nodeColor=<rgba 0 0.6667 1 1>}
  n21 = OneMinus(None=VertexColor().G)
  n22 = DynamicParameter() {defaultValue=<rgba 1 0 1 0>, paramNames [Gravity Offset, Distortion Amount, Distortion Time, Erode]}
  n23 = Saturate(None=Subtract(A=Add(A=@n21), B=Multiply(A=@n22.Erode) {constB=2}))
  n24 = ScalarParameter() {defaultValue=2, group=Opacity, parameterName=Fresnel Power}
  n25 = OneMinus(None=Fresnel(ExponentIn=Multiply(A=@n23, B=@n24), BaseReflectFractionIn=ScalarParameter() {group=Opacity, parameterName=Fresnel Minimum Opacity}))
  n26 = ScalarParameter() {defaultValue=1, parameterName=Param_3}
  NamedRerouteDeclaration(None=Multiply(
    A=@n2
    B=MaterialFunctionCall(In=@n25, Contrast=@n26) {materialFunction=&CheapContrast}.Result
   )) {name=Opacity, nodeColor=<rgba 1 1 1 1>}
  NamedRerouteDeclaration(None=Multiply(A=VectorParameter() {group=Color, parameterName=Subsurface Color, defaultValue=<rgba 0.25 0.75 1 1>}.RGB, B=ParticleColor().RGB)) {name=Subsurface Color, nodeColor=<rgba 1 0 0.3984 1>}
  n27 = MaterialFunctionCall(Float3=RgbToHsv!(None=ParticleColor().RGB)) {materialFunction=&BreakOutFloat3Components}
  n28 = ScalarParameter() {defaultValue=1.068, parameterName=ShadowSaturation}
  n29 = CollectionParameter() {
   collection=&FortniteMaterialParameters
   parameterName=SunAndMoonModelDirectionalVector
  }
  n30 = Saturate(None=DotProduct(A=VertexNormalWS(), B=@n29))
  n31 = Fresnel(ExponentIn=ScalarParameter() {defaultValue=1, parameterName=Diffuse Fresnel Exp}) {baseReflectFraction=0, exponent=1}
  n32 = Multiply(
   A=OneMinus(None=Multiply(A=@n30, B=@n30))
   B=StaticSwitchParameter(True=OneMinus(None=@n31), False=@n31) {parameterName=Flip Diffuse Fresnel}
  )
  n33 = ScalarParameter() {defaultValue=0.5, parameterName=ShadowValue}
  n34 = MaterialFunctionCall(
   X=@n27.R
   Y=LinearInterpolate(A=@n27.G, B=Multiply(A=@n27.G, B=@n28), Alpha=@n32) {constB=0.95}
   Z=LinearInterpolate(A=@n27.B, B=Multiply(A=@n27.G, B=@n33), Alpha=@n32) {constB=0.5}
  ) {materialFunction=&MakeFloat3}
  NamedRerouteDeclaration(None=Add(A=HsvToRgb!(None=@n34.Result), B=NamedRerouteUsage())) {name=Diffuse Color, nodeColor=<rgba 1 0.25 0 1>}
  n35 = Add(A=ParticleRandom(), B=LocalPosition() {localOrigin=InstancePreSkinning}.XYZ)
  n36 = ScalarParameter() {defaultValue=0.005, parameterName=Distortion Size}
  n37 = VectorParameter() {parameterName=Distortion Direction, defaultValue=<rgba -0.1 0 0 1>}
  n38 = DynamicParameter() {defaultValue=<rgba 1 0 1 0>, paramNames [Gravity Offset, Distortion Amount, Distortion Time, Erode]}
  n39 = TextureSample(UVs=Add(A=Multiply(A=@n35, B=@n36), B=Multiply(A=@n37.RGB, B=@n38.Distortion Time))) {samplerType=SAMPLERTYPE_LinearColor, texture=/Volumetrics/Content/VolumeTextures/Textures/VT_CurlNoise}
  n40 = DynamicParameter() {defaultValue=<rgba 1 0 1 0>, paramNames [Gravity Offset, Distortion Amount, Distortion Time, Erode]}
  n41 = Constant()
  n42 = AppendVector(A=AppendVector(A=@n41, B=@n41), B=@n40.Gravity Offset)
  n43 = OneMinus(None=@n4)
  NamedRerouteDeclaration(None=Add(A=Multiply(A=@n39.RGB, B=@n40.Distortion Amount), B=Multiply(A=@n42, B=@n43))) {name=WPO, nodeColor=<rgba 1 0.5391 0 1>}
  n44 = Multiply(
   A=ComponentMask(None=VectorParameter() {group=Added Texture, parameterName=AddedTextureColor, defaultValue=<rgba 0.9531 0.9531 0.9531 0>}.RGB) {b=true, g=true, r=true}
   B=ParticleColor().A
  )
  n45 = TextureCoordinate()
  n46 = DynamicParameter() {parameterIndex=1, defaultValue=<rgba 1 1 0 0>, paramNames [TexTileX,TexTileY,TexOffsetX,TexOffsetY]}
  n47 = AppendVector(A=@n46.TexTileX, B=@n46.TexTileX)
  n48 = TextureSampleParameter2D(UVs=Add(A=Multiply(A=@n45, B=@n47), B=AppendVector(A=@n46.TexOffsetX, B=@n46.TexOffsetX))) {group=Added Texture, parameterName=AddedTexture, texture=/DealershipSharedLibrary/FX/Textures/Masks/T_VCVFX_PlasmaRibbonMask02_RGB}
  n49 = ChannelMaskParameter(None=@n48.RGB) {group=Added Texture, maskChannel=Green, parameterName=AddedTextureChannels, defaultValue=<rgba 0 1 0 0>}
  n50 = Multiply(
   A=Power(Base=@n49.R, Exp=ScalarParameter() {defaultValue=1, group=Added Texture, parameterName=TextureCoverage})
   B=ScalarParameter() {defaultValue=1, group=Added Texture, parameterName=TextureStrength}
  )
  n51 = Constant4Vector()
  NamedRerouteDeclaration(None=StaticSwitchParameter(True=LinearInterpolate(A=@n44, Alpha=@n50) {constB=0}, False=ComponentMask(None=@n51.RGBA) {b=true, g=true, r=true}) {parameterName=UseAddedTexture?}) {name=AddedTexture, nodeColor=<rgba 1 0.6797 0 1>}
  NamedRerouteDeclaration(None=LinearInterpolate(B=ScalarParameter() {defaultValue=1.3, group=Refraction, parameterName=Refraction Value}) {constA=1}) {name=Refraction, nodeColor=<rgba 1 1 1 1>}
  NamedRerouteUsage()
  NamedRerouteUsage()
  NamedRerouteUsage()
  NamedRerouteUsage()
  NamedRerouteUsage()
 ]
}
