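# Translucent subsurface master material for water splash particles on Niagara sprites, ribbons and meshes, and on particle sprites.
# Dynamic parameters drive noise-distorted UVs, foam blending and threshold erosion of the opacity mask, combined with near-camera, depth and fresnel fades.
# Emissive fakes lighting with a view-space reflection texture and sun specular, clamped higher on mobile; refraction is also computed.
# Use_MeltPool blends to a puddle variant with decal textures and noise color.
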Material /DealershipSharedLibrary/FX/Materials/MasterMaterials/Liquid/M_DLR_Water_Splash
graph {
 properties {
  bEnableResponsiveAA=true
  bUsedWithNiagaraMeshParticles=true
  bUsedWithNiagaraRibbons=true
  bUsedWithNiagaraSprites=true
  bUsedWithParticleSprites=true
  blendMode=BLEND_Translucent
  pixelDepthOffsetMode=PDOM_Legacy
  refractionMethod=RM_IndexOfRefraction
  shadingModel=MSM_Subsurface
  translucentSelfShadowDensityScale=0.1
  translucentShadowDensityScale=0.1
  usedShadingModels=Subsurface
 }
 nodes [
  n1 = MaterialFunctionCall(Float2=Multiply(
    A=MaterialFunctionCall() {materialFunction=&VectorToRadialValue}.Radial Coordinates
    B=Constant2Vector() {g=1, r=2}
   )) {materialFunction=&BreakOutFloat2Components}
  Add(A=@n1.G, B=Multiply(A=Time()) {constB=-0.1})
  Constant3Vector()
  LinearInterpolate(A=NamedRerouteUsage(), B=NamedRerouteUsage(), Alpha=ScalarParameter() {parameterName=Use_MeltPool})
  LinearInterpolate(A=NamedRerouteUsage(), B=NamedRerouteUsage(), Alpha=ScalarParameter() {parameterName=Use_MeltPool})
  LinearInterpolate(A=NamedRerouteUsage(), B=Multiply(A=NamedRerouteUsage(), B=NamedRerouteUsage()), Alpha=ScalarParameter() {parameterName=Use_MeltPool})
  LinearInterpolate(A=NamedRerouteUsage(), B=Constant() {r=1}, Alpha=ScalarParameter() {parameterName=Use_MeltPool})
  LinearInterpolate(
   A=NamedRerouteUsage()
   B=LinearInterpolate(B=NamedRerouteUsage(), Alpha=NamedRerouteUsage())
   Alpha=ScalarParameter() {parameterName=Use_MeltPool}
  )
  MaterialFunctionCall(Emissive=LinearInterpolate(A=NamedRerouteUsage(), B=NamedRerouteUsage(), Alpha=ScalarParameter() {parameterName=Use_MeltPool})) {materialFunction=&MF_FX_EmissiveAdjustments}
  n2 = Multiply(
   A=LinearInterpolate(
    A=VectorParameter() {group=Color, parameterName=Light Color, defaultValue=<rgba 0.33 0.4675 0.66 0>}.RGB
    B=VectorParameter() {group=Color, parameterName=Dark Color, defaultValue=<rgba 0.2025 0.3056 0.45 0>}.RGB
    Alpha=NamedRerouteUsage()
   ) {constA=1, constB=0.75}
   B=ParticleColor().RGB
  )
  n3 = ScalarParameter() {defaultValue=0.66, group=Foam, parameterName=Foam - Desaturation}
  n4 = Multiply(
   A=MaterialFunctionCall(
    Hue Shift Percentage=ScalarParameter() {group=Foam, parameterName=Foam - Hue Shift, sliderMax=1}
    Texture=Desaturation(None=@n2, Fraction=@n3)
   ) {materialFunction=&HueShift}.Result
   B=ScalarParameter() {defaultValue=10, group=Foam, parameterName=Foam - Value Multiplier}
  )
  n5 = NamedRerouteUsage()
  MaterialFunctionCall!(
   Vector3=NamedRerouteDeclaration(None=LinearInterpolate(A=@n2, B=@n4, Alpha=@n5)) {name=Diffuse Color, nodeColor=<rgba 1 0.25 0 1>}
   MaximumNumberOfDigits
   UVs
   DebugTextLocation RG_UpperRight BA_LowerLeft
   Component Spacing
  ) {materialFunction=/Engine/Functions/Engine_MaterialFunctions02/Utility/DebugFloat3Values}
  n6 = NamedRerouteUsage()
  n7 = Constant() {r=0.5}
  n8 = Multiply(A=Transform(None=@n6) {transformType=TRANSFORM_View}, B=@n7)
  n9 = TextureSample(UVs=ComponentMask(None=Add(A=@n8, B=@n7)) {g=true, r=true}) {texture=/Game/VisualThreatManager/StormVisuals/LargeClouds/T_EyeballReflection2}
  n10 = ScalarParameter() {defaultValue=0.2, group=Specular, parameterName=Reflection Amount}
  n11 = ScalarParameter() {
   defaultValue=0.075
   desc="Offset the Minimum fake specular amount for Mobile.  ("
   group=Specular
   parameterName=Mobile Specular Min Offset
   sliderMax=0.2
  }
  n12 = FeatureLevelSwitch(Default=Constant(), ES3_1=@n11)
  n13 = Clamp(
   None=Desaturation(None=Multiply(A=@n9.RGB, B=@n10), Fraction=ScalarParameter() {group=Specular, parameterName=Reflection - Desaturation})
   Min=@n12
  ) {minDefault=0.1}
  Multiply(A=@n13, B=MaterialFunctionCall() {materialFunction=&Color_Time_of_Day}.Result)
  n14 = Multiply(A=NamedRerouteUsage(), B=ScalarParameter() {defaultValue=0.1, parameterName=DiffuseColorEmissiveMultiply})
  n15 = CollectionParameter() {
   collection=&FortniteMaterialParameters
   parameterName=SunAndMoonModelDirectionalVector
  }
  n16 = Add(
   A=Multiply(
    A=VectorParameter() {parameterName=FakeSunDirection, defaultValue=<rgba 0 -15 50 1>}.RGB
    B=ScalarParameter() {parameterName=Fake Sun Direction Amount}
   )
   B=Multiply(A=ComponentMask(None=@n15) {b=true, g=true, r=true}, B=ScalarParameter() {defaultValue=1, parameterName=Sun Direction Amount}) {constB=0}
  )
  n17 = ReflectionVectorWS(CustomWorldNormal=Transform(None=@n6)) {bNormalizeCustomWorldNormal=true}
  n18 = NamedRerouteUsage()
  n19 = Power(
   Base=Saturate(None=DotProduct(A=@n16, B=@n17))
   Exp=LinearInterpolate(
    A=ScalarParameter() {defaultValue=40, group=Specular, parameterName=Specular Power}
    B=ScalarParameter() {defaultValue=5, group=Foam, parameterName=Foam - Specular Power}
    Alpha=@n18
   )
  ) {constExponent=40}
  n20 = LinearInterpolate(
   A=ScalarParameter() {defaultValue=5, group=Specular, parameterName=Specular Intensity}
   B=ScalarParameter() {defaultValue=1, group=Foam, parameterName=Foam - Specular Intensity}
   Alpha=@n18
  )
  n21 = CollectionParameter() {collection=&FortniteMaterialParameters, parameterName=SunLightColor}
  n22 = Clamp(None=Power(Base=ComponentMask(None=@n21) {b=true, g=true, r=true}) {constExponent=0.1})
  n23 = VectorParameter() {parameterName=FakeSunLighColor, defaultValue=<rgba 1 1 1 1>}
  n24 = Add(A=@n13, B=Multiply(A=Clamp(None=@n19, Min=@n12, Max=@n20), B=Multiply(A=@n22, B=@n23.RGB))) {constB=0}
  NamedRerouteDeclaration(None=StaticSwitchParameter(True=Add(A=@n14, B=@n24), False=@n24) {parameterName=DiffuseColor on Emissive?}) {name=Emissive, nodeColor=<rgba 0 0.6667 1 1>}
  n25 = TextureSampleParameter2D(UVs=NamedRerouteUsage()) {parameterName=Opacity Mask, samplerType=SAMPLERTYPE_LinearColor, texture=/Game/Effects/Fort_Effects/Textures/Water/T_waterSplash_D}
  n26 = ChannelMaskParameter(None=LinearInterpolate(
    A=@n25.RGBA
    B=TextureSample() {texture=/Game/Effects/Niagara/MasterMaterials/Liquid/T_FX_Emote_SatinCheddar_Splash_Decal}.RGBA
    Alpha=ScalarParameter() {parameterName=Use_MeltPool}
   )) {maskChannel=Green, parameterName=Opacity MasKChannel, defaultValue=<rgba 0 1 0 0>}
  n27 = LinearInterpolate(Alpha=Saturate(None=DynamicParameter() {defaultValue=<rgba 1 0 1 0>, paramNames [Rand 0-1, Foam, UV Distortion, Erode]}.Erode)) {constA=0.04}
  MaterialExpressionSaturate_8 = Saturate(None=Divide(A=Subtract(A=@n26.R, B=@n27), B=OneMinus(None=@n27)))
  MaterialExpressionScalarParameter_34 = ScalarParameter() {defaultValue=0, parameterName=ThresholdParam}
  n28 = Custom!(Opacity=@MaterialExpressionSaturate_8, Threshold=@MaterialExpressionScalarParameter_34) {
   code="clip(Opacity - Threshold);\r\nreturn Opacity;"
   description=ClipMask
   outputType=CMOT_Float1
   inputs [
    {
     inputName=Opacity
     input {expression=&self:MaterialExpressionSaturate_8}
    }
    {
     inputName=Threshold
     input {expression=&self:MaterialExpressionScalarParameter_34}
    }
   ]
  }
  n29 = NamedRerouteDeclaration(None=@n28) {name=Eroded Mask, nodeColor=<rgba 1 0 0.9844 1>}
  n30 = ScalarParameter() {defaultValue=0.1, group=Opacity, parameterName=OpacitySoftness}
  n31 = ScalarParameter() {defaultValue=40, group=Opacity, parameterName=NearFade}
  n32 = Multiply(
   A=Power(
    Base=MaterialFunctionCall(Distance=@n31) {materialFunction=&Near_Camera_Fade}.Result
    Exp=ScalarParameter() {defaultValue=3, group=Opacity, parameterName=NearFadeExp}
   )
   B=ParticleColor().A
  )
  n33 = StaticSwitchParameter(True=@n32, False=DepthFade(Opacity=@n32, FadeDistance=ScalarParameter() {defaultValue=2, group=Opacity, parameterName=Depth Fade Distance})) {parameterName=Masked Mode}
  n34 = Multiply(
   A=Saturate(None=Divide(A=@n29, B=@n30))
   B=Multiply(A=NamedRerouteDeclaration(None=@n33) {name=Alpha Multiplier, nodeColor=<rgba 0 0.3672 1 1>}, B=NamedRerouteUsage())
  )
  n35 = ScalarParameter() {defaultValue=-3, parameterName=Depth decal High}
  n36 = StaticSwitchParameter(True=@n35, False=LinearInterpolate(A=ScalarParameter() {defaultValue=1, parameterName=Depth decal Low}, B=@n35, Alpha=DepthFade())) {parameterName=Masked Mode}
  n37 = ScalarParameter() {parameterName=Use Decal Like Depthfade}
  NamedRerouteDeclaration(None=Multiply(A=Saturate(None=@n34), B=LinearInterpolate(B=@n36, Alpha=@n37) {constA=1})) {name=Opacity, nodeColor=<rgba 1 1 1 1>}
  NamedRerouteDeclaration(None=Multiply(A=VectorParameter() {group=Color, parameterName=Subsurface Color, defaultValue=<rgba 0.25 0.75 1 1>}.RGB, B=ParticleColor().RGB)) {name=Subsurface Color, nodeColor=<rgba 1 0 0.3984 1>}
  n38 = ScalarParameter() {defaultValue=1, group=Opacity, parameterName=Fresnel Minimum Opacity}
  n39 = NamedRerouteUsage()
  n40 = OneMinus(None=NamedRerouteUsage())
  n41 = ScalarParameter() {defaultValue=3, group=Opacity, parameterName=Fresnel Power}
  NamedRerouteDeclaration(None=LinearInterpolate(A=LinearInterpolate(A=@n38, Alpha=@n39), Alpha=Power(Base=@n40, Exp=@n41)) {constA=0.6}) {name=Fake Fresnel Opacity, nodeColor=<rgba 0 0.6484 1 1>}
  n42 = MaterialFunctionCall(Float2=TextureCoordinate()) {materialFunction=&BreakOutFloat2Components}
  n43 = TextureSample!(UVs=AppendVector(A=@n1.R, B=@n1.R), Tex, "DDX(UVs)"=DDX(Value=@n42.R), "DDY(UVs)"=DDY(Value=@n42.G), Apply View MipBias) {
   mipValueMode=TMVM_Derivative
   samplerType=SAMPLERTYPE_Normal
   texture=/Game/Characters/Player/Female/Medium/Bodies/F_MED_Bistro_Astronaut/FX/T_GaussianNoise_Normal
  }
  n44 = LinearInterpolate(
   A=@n43.RGB
   B=Constant3Vector() {constant=<rgba 0 0 1 0>}
   Alpha=MaterialFunctionCall!(Hardness, Center, Radius=Constant() {r=0.2}) {materialFunction=/Game/Packages/Fortress_SharedMaterials/Base_Material_Functions/FX_SphereMask2D}.Result
  )
  n45 = Constant() {r=0.2}
  NamedRerouteDeclaration(None=MaterialFunctionCall(
    BaseNormal=TextureSample!(UVs, Tex, Apply View MipBias) {samplerType=SAMPLERTYPE_Normal, texture=/Game/Effects/Niagara/MasterMaterials/Liquid/T_FX_Emote_SatinCheddar_Splash_Decal_Normal}.RGB
    AdditionalNormal=MaterialFunctionCall(Normal=@n44, Flatness=@n45) {materialFunction=&FlattenNormal}.Result
   ) {materialFunction=&BlendAngleCorrectedNormals}.Result) {name=MeltPool_Normal, nodeColor=<rgba 0 1 0.5781 1>}
  n46 = TextureSample() {texture=&T_FX_Emote_SatinCheddar_Splash_Decal}
  n47 = OneMinus(None=ParticleColor().A)
  n48 = Multiply(A=LinearInterpolate(A=LinearInterpolate(Alpha=@n47) {constB=0.1}, Alpha=OneMinus(None=@n46.G)) {constA=0.1}, B=@n47)
  n49 = Multiply(A=Subtract(A=@n46.R, B=Multiply(A=@n48) {constB=10}), B=LinearInterpolate(Alpha=@n47) {constA=1, constB=2})
  n50 = Constant() {r=1}
  n51 = DepthFade(FadeDistance=ScalarParameter() {parameterName=Melt Pool Decal Like Depth Fade Distance})
  n52 = Constant() {r=1}
  n53 = If(
   A=ScalarParameter() {parameterName=Melt Pool Decal Like Depth Fade}
   "A > B"=OneMinus(None=StaticSwitchParameter(True=@n50, False=@n51) {parameterName=Masked Mode})
   "A == B"=@n52
   "A < B"=@n52
  ) {constB=0.99}
  n54 = Constant() {r=1}
  NamedRerouteDeclaration(None=Multiply(
    A=Saturate(None=@n49)
    B=MaterialFunctionCall!(Default=@n53, Low/Mobile=@n54) {materialFunction=/Game/Effects/Fort_Effects/Materials/_Functions/MF_LowMobile_PathingSwitch}.Result
   )) {name=MeltPool_Opacity, nodeColor=<rgba 0 0.2266 1 1>}
  n55 = NamedRerouteUsage()
  n56 = Constant() {r=0.5}
  n57 = Multiply(A=Transform(None=@n55) {transformType=TRANSFORM_View}, B=@n56)
  n58 = TextureSample(UVs=ComponentMask(None=Add(A=@n57, B=@n56)) {g=true, r=true}) {texture=&T_EyeballReflection2}
  n59 = ScalarParameter() {defaultValue=0.2, group=Specular, parameterName=Reflection Amount}
  n60 = ScalarParameter() {
   defaultValue=0.075
   desc="Offset the Minimum fake specular amount for Mobile.  ("
   group=Specular
   parameterName=Mobile Specular Min Offset
   sliderMax=0.2
  }
  n61 = Clamp(
   None=Desaturation(None=Multiply(A=@n58.RGB, B=@n59), Fraction=ScalarParameter() {group=Specular, parameterName=Reflection - Desaturation})
   Min=FeatureLevelSwitch(Default=Constant(), ES3_1=@n60)
  ) {minDefault=0.1}
  n62 = MaterialFunctionCall() {materialFunction=&Color_Time_of_Day}
  n63 = CollectionParameter() {
   collection=&FortniteMaterialParameters
   parameterName=SunAndMoonModelDirectionalVector
  }
  n64 = ComponentMask(None=@n63) {b=true, g=true, r=true}
  n65 = ReflectionVectorWS(CustomWorldNormal=Transform(None=@n55)) {bNormalizeCustomWorldNormal=true}
  n66 = Power(Base=Saturate(None=DotProduct(A=@n64, B=@n65))) {constExponent=40}
  n67 = Power(Base=ComponentMask(None=CollectionParameter() {collection=&FortniteMaterialParameters, parameterName=SunLightColor}) {b=true, g=true, r=true}) {constExponent=0.1}
  n68 = Multiply(A=Multiply(A=Clamp(None=@n66), B=Clamp(None=@n67)), B=NamedRerouteUsage())
  NamedRerouteDeclaration(None=Add(A=Multiply(A=@n61, B=@n62.Result), B=Multiply(A=@n68) {constB=0.2}) {constB=0}) {name=ReflectionToEmissive, nodeColor=<rgba 0.8281 0 1 1>}
  n69 = Multiply(
   A=Multiply(A=TextureCoordinate(), B=ScalarParameter() {defaultValue=8, group=UV, parameterName=Distortion Tile})
   B=MaterialFunctionCall(X=ScalarParameter() {defaultValue=1, group=UV, parameterName="Distortion U Tile (Ribbon)"}, Y=Constant() {r=1}) {materialFunction=&MakeFloat2}.Result
  )
  n70 = Add(A=DynamicParameter() {defaultValue=<rgba 1 0 1 0>, paramNames [Rand 0-1, Foam, UV Distortion, Erode]}.Rand 0-1, B=ParticleRelativeTime())
  n71 = Multiply(A=Constant2Vector() {g=-0.8, r=0.05}, B=ScalarParameter() {defaultValue=1, group=UV, parameterName=Distortion Speed})
  n72 = ComponentMask(None=TextureSample(UVs=Panner(Coordinate=@n69, Time=@n70, Speed=@n71) {speedX=0.05, speedY=0.5}) {samplerType=SAMPLERTYPE_Normal, texture=&T_Weap_Distort_N_02}.RGB) {g=true, r=true}
  n73 = DynamicParameter() {defaultValue=<rgba 1 0 1 0>, paramNames [Rand 0-1, Foam, UV Distortion, Erode]}
  n74 = Multiply(
   A=Multiply(A=ScalarParameter() {defaultValue=0.02, group=UV, parameterName=Distortion Amount}, B=@n73.UV Distortion)
   B=LinearInterpolate(Alpha=@n73.Rand 0-1) {constA=0.5, constB=1.2}
  )
  NamedRerouteDeclaration(None=Add(A=TextureCoordinate(), B=Multiply(A=@n72, B=@n74))) {name=UVs, nodeColor=<rgba 1 1 0 1>}
  n75 = TextureSampleParameter2D(UVs=NamedRerouteUsage()) {parameterName=Normal, samplerType=SAMPLERTYPE_Normal, texture=/Game/Effects/Fort_Effects/Textures/Water/T_waterSplash_R_N}
  n76 = MaterialFunctionCall(X=@n75.R, Y=@n75.R, Z=NamedRerouteUsage()) {materialFunction=&MakeFloat3}
  NamedRerouteDeclaration(None=MaterialFunctionCall(Normal=Normalize(VectorInput=@n76.Result), Flatness=ScalarParameter() {group=Normal, parameterName=FlattenNormal}) {materialFunction=&FlattenNormal}.Result) {name=Normal, nodeColor=<rgba 0 0.25 1 1>}
  n77 = NamedRerouteUsage()
  n78 = NamedRerouteUsage()
  NamedRerouteDeclaration(None=LinearInterpolate(B=ScalarParameter() {defaultValue=1.3, group=Refraction, parameterName=Refraction Value}, Alpha=Multiply(A=@n77, B=@n78)) {constA=1}) {name=Refraction, nodeColor=<rgba 1 1 1 1>}
  n79 = DynamicParameter() {defaultValue=<rgba 1 0 1 0>, paramNames [Rand 0-1, Foam, UV Distortion, Erode]}
  n80 = TextureCoordinate()
  n81 = ScalarParameter() {defaultValue=2.5, group=Foam, parameterName=Foam - Mask Tile}
  n82 = ScalarParameter() {defaultValue=1, group=Foam, parameterName="Foam - Mask U Tile (Ribbon)"}
  n83 = Constant() {r=1}
  n84 = TextureSampleParameter2D(UVs=Multiply(
    A=Multiply(A=@n80, B=@n81)
    B=MaterialFunctionCall(X=@n82, Y=@n83) {materialFunction=&MakeFloat2}.Result
   )) {
   group=Foam
   parameterName=Foam Mask
   samplerType=SAMPLERTYPE_LinearGrayscale
   texture=/Game/Global/Textures/Default/Noise/T_SwirlNoise_N_height
  }
  n85 = Divide(
   A=Add(
    A=LinearInterpolate(Alpha=@n79.Foam) {constA=-1}
    B=ChannelMaskParameter(None=@n84.RGB) {group=Foam, maskChannel=Green, parameterName=Foam Mask Channel, defaultValue=<rgba 0 1 0 0>}.R
   )
   B=ScalarParameter() {defaultValue=0.75, group=Foam, parameterName=Foam - Mask Softness}
  ) {constB=0.25}
  NamedRerouteDeclaration(None=LinearInterpolate(B=Saturate(None=@n85), Alpha=ScalarParameter() {group=Foam, parameterName=Enable Foam})) {name=Foam, nodeColor=<rgba 0 0.2734 1 1>}
  n86 = Constant3Vector() {constant=<rgba 0.0401 0.3606 0.7708 1>}
  n87 = Constant3Vector() {constant=<rgba 0.4375 0.821 1 1>}
  n88 = Panner()
  n89 = TextureSample(UVs=@n88) {samplerType=SAMPLERTYPE_LinearColor, texture=/Game/Global/Textures/Default/Noise/T_Noise01_Inverse}
  n90 = TextureSample(UVs=Divide(A=@n88)) {samplerType=SAMPLERTYPE_LinearColor, texture=&T_Noise01_Inverse}
  n91 = Add(A=Multiply(A=TextureCoordinate(), B=Constant() {r=2}), B=Divide(A=Multiply(A=@n89.R, B=@n90.R)) {constB=1})
  n92 = TextureSample(UVs=@n91) {texture=/Game/Weapons/FORT_Melee/Pickaxe_MadameMoth/FX/T_Fibre_Noise}
  n93 = Divide(A=@n91)
  n94 = TextureSample(UVs=Divide(A=@n91) {constB=1.5}) {texture=/Game/Global/Textures/Default/Noise/T_Tiling_Noise_01}
  n95 = Multiply(
   A=LinearInterpolate(
    A=LinearInterpolate(A=@n86, B=@n87, Alpha=@n92.R)
    B=Constant3Vector() {constant=<rgba 0.6519 0.664 0.7031 1>}
    Alpha=TextureSample(UVs=@n93) {texture=&T_Fibre_Noise}.R
   )
   B=@n94.R
  )
  NamedRerouteDeclaration(None=Add(A=Divide(A=@n95) {constB=5}, B=NamedRerouteUsage())) {name=MeltPool_Emissive, nodeColor=<rgba 1 0.6797 0 1>}
  n96 = Constant3Vector() {constant=<rgba 0.2917 0.6864 1 1>}
  n97 = Constant3Vector() {constant=<rgba 0.0401 0.3606 0.7708 1>}
  n98 = Multiply(A=@n94.R, B=NamedRerouteUsage())
  n99 = Constant3Vector() {constant=<rgba 1 1 1 0>}
  n100 = Constant3Vector() {constant=<rgba 0.9 0.9 1.35 0>}
  NamedRerouteDeclaration(None=Multiply(
    A=LinearInterpolate(A=@n96, B=@n97, Alpha=@n98)
    B=MaterialFunctionCall(Color_Morning=@n99, Color_Day=@n100, Color_Evening=@n99, Color_Night=@n99) {materialFunction=&Color_Time_of_Day}.Result
   )) {name=MeltPool_BaseColor, nodeColor=<rgba 1 0.5156 0 1>}
  TextureSample() {samplerType=SAMPLERTYPE_Masks, texture=/Game/Global/Textures/Default/Noise/T_Metal_Noise}
  TextureSample() {texture=&T_Fibre_Noise}
  TextureSample() {texture=&T_Tiling_Noise_01}
 ]
}
